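# Translucent lit Niagara master material for lava splash effects on sprites, ribbons and mesh particles.
# Two panning noise normals distort the UVs of the splash mask and normal map.
# The mask is thresholded by a Step value with optional rounding, optionally eroded by HeightLerp noise dissolve, then depth-faded, with dynamic parameters driving step, dissolve and base color brightness.
# A powered mask blends dark-to-light emissive, and temporal dither feeds a depth-offset term.
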Material /BRCosmetics/Effects/Niagara/MasterMaterials/Splash/M_Lava_Splash
graph {
 properties {
  bUsedWithNiagaraMeshParticles=true
  bUsedWithNiagaraRibbons=true
  bUsedWithNiagaraSprites=true
  blendMode=BLEND_Translucent
  pixelDepthOffsetMode=PDOM_Legacy
  usedShadingModels=Default Lit
 }
 nodes [
  n1 = DynamicParameter() {defaultValue=<rgba 0 1 0 1>, paramNames [Step,BaseColor,Dissolve,Param4]}
  n2 = VectorParameter() {parameterName=DissolveUV_Scale, defaultValue=<rgba 1 1 0 0>}
  n3 = TextureSampleParameter2D(UVs=Multiply(A=TextureCoordinate(), B=ComponentMask(None=@n2.RGB) {g=true, r=true})) {parameterName=Dissolve_Tex, samplerType=SAMPLERTYPE_LinearGrayscale, texture=/Game/Global/Textures/Default/Noise/T_SwirlNoise_N_height}
  n4 = TextureSample(UVs=Panner(Coordinate=TextureCoordinate()) {speedX=0.25, speedY=-0.1}) {samplerType=SAMPLERTYPE_Normal, texture=/Game/Characters/Player/Female/Medium/Bodies/F_MED_Bistro_Astronaut/FX/T_GaussianNoise_Normal}
  n5 = TextureSample(UVs=Panner(Coordinate=TextureCoordinate() {uTiling=0.7, vTiling=0.7}) {speedX=-0.15, speedY=0.15}) {samplerType=SAMPLERTYPE_Normal, texture=&T_GaussianNoise_Normal}
  n6 = ComponentMask(None=Multiply(A=@n4.RGB, B=@n5.RGB)) {g=true, r=true}
  n7 = ScalarParameter() {defaultValue=0.2, group=Noise, parameterName=NoiseStrength}
  n8 = Add(
   A=TextureCoordinate()
   B=Add(A=Multiply(A=@n6, B=@n7) {constB=0.2}, B=ScalarParameter() {defaultValue=-0.02, group=Noise, parameterName=NoiseOffset}) {constB=-0.02}
  )
  n9 = TextureObjectParameter() {
   group=Splash
   parameterName=SplashTex
   samplerType=SAMPLERTYPE_LinearGrayscale
   texture=/Game/Characters/Player/Female/Medium/Bodies/F_MED_Bistro_Astronaut/FX/T_Water_Splash_Alpha_01
  }
  n10 = ChannelMaskParameter(None=TextureSample(UVs=@n8, Tex=@n9).RGBA) {group=Splash, parameterName=SplashMaskChannel}
  n11 = ScalarParameter() {parameterName=Step}
  n12 = DynamicParameter() {defaultValue=<rgba 0 1 0 1>, paramNames [Step,BaseColor,Dissolve,Param4]}
  n13 = ScalarParameter() {defaultValue=1, parameterName=Use Dynamic Step}
  n14 = Multiply(
   A=Subtract(A=@n10.R, B=LinearInterpolate(A=@n11, B=@n12.Step, Alpha=@n13))
   B=ScalarParameter() {defaultValue=1, parameterName=AlphaIntensity}
  ) {constB=3}
  n15 = Multiply(
   A=LinearInterpolate(A=@n14, B=Round(None=@n14), Alpha=ScalarParameter() {defaultValue=1, parameterName=UseRound})
   B=LinearInterpolate(
    B=MaterialFunctionCall(Transition Phase=@n1.Dissolve, Height Texture=@n3.R) {materialFunction=&HeightLerp}.Results
    Alpha=ScalarParameter() {parameterName=UseCostumDissolve}
   ) {constA=1}
  )
  DepthFade(Opacity=Multiply(A=ParticleColor().A, B=@n15), FadeDistance=ScalarParameter() {parameterName=FadeDistance})
  n16 = Multiply(A=VectorParameter() {parameterName=BaseColor, defaultValue=<rgba 0.4 0.0596 0.0417 1>}.RGB, B=@n12.BaseColor)
  LinearInterpolate(A=@n16, B=Multiply(A=ParticleColor().RGB, B=@n16), Alpha=ScalarParameter() {parameterName=UseBaseIntensity})
  n17 = TextureObjectParameter() {
   group=Splash
   parameterName=NormalTex
   samplerType=SAMPLERTYPE_Normal
   texture=/Game/Characters/Player/Female/Medium/Bodies/F_MED_Bistro_Astronaut/FX/T_Water_Splash_Normal_01
  }
  MaterialFunctionCall(Normal=TextureSample(UVs=@n8, Tex=@n17).RGB, Flatness=ScalarParameter() {defaultValue=-2, group=Splash, parameterName=NormalStrength}) {materialFunction=&FlattenNormal}
  n18 = VectorParameter() {parameterName=Emissive Dark Color, defaultValue=<rgba 0.02 0.001 0 1>}
  n19 = VectorParameter() {parameterName=Emissive Light Color, defaultValue=<rgba 0.333 0.0383 0.0333 1>}
  n20 = Power(Base=@n10.R, Exp=ScalarParameter() {defaultValue=5, group=Splash, parameterName=Emissive Power})
  MaterialFunctionCall(Emissive=Multiply(A=LinearInterpolate(A=@n18.RGB, B=@n19.RGB, Alpha=@n20), B=ParticleColor().RGB)) {materialFunction=&MF_FX_EmissiveAdjustments}
  Multiply(
   A=ScalarParameter() {parameterName=DitherDepthThershold}
   B=MaterialFunctionCall() {materialFunction=&DitherTemporalAA}.Result
  )
  Reroute(None=@n15)
  ScalarParameter() {defaultValue=0.2, parameterName=Roughness}
 ]
}
